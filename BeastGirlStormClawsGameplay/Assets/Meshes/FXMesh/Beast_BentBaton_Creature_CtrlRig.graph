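# Control Rig for the Beast BentBaton creature FX mesh that aims to orient the left and right claw FX bones with the hands.
# Two chained local-space-offset rotation constraints make each claw_FX_01 bone follow the equally weighted average of its side's four finger metacarpals.
# The construction graph imports the skeleton hierarchy.
# The constraint chain has no link from the Begin Execution node.
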ControlRigBlueprint /BeastGirlStormClawsGameplay/Assets/Meshes/FXMesh/Beast_BentBaton_Creature_CtrlRig
graphs [
 RigVMFunctionLibrary {links [], nodes []}
 RigVMModel {nodes [
   &RigUnit_RotationConstraintLocalSpaceOffset(
    AdvancedSettings=(InterpolationType=Average,RotationOrderForFilter=XZY)
    Child=(Type=Bone,Name="claw_FX_01_l")
    ExecutePin=()
    Filter=(bX=True,bY=True,bZ=True)
    Parents=((Item=(Type=Bone,Name="index_metacarpal_l"),Weight=1),(Item=(Type=Bone,Name="middle_metacarpal_l"),Weight=1),(Item=(Type=Bone,Name="ring_metacarpal_l"),Weight=1),(Item=(Type=Bone,Name="pinky_metacarpal_l"),Weight=1))
    Weight=1
   )
   > &RigUnit_RotationConstraintLocalSpaceOffset(
    AdvancedSettings=(InterpolationType=Average,RotationOrderForFilter=XZY)
    Child=(Type=Bone,Name="claw_FX_01_r")
    Filter=(bX=True,bY=True,bZ=True)
    Parents=((Item=(Type=Bone,Name="index_metacarpal_r"),Weight=1),(Item=(Type=Bone,Name="middle_metacarpal_r"),Weight=1),(Item=(Type=Bone,Name="ring_metacarpal_r"),Weight=1),(Item=(Type=Bone,Name="pinky_metacarpal_r"),Weight=1))
    Weight=1
   )
   RigUnit_BeginExecution = &RigUnit_BeginExecution()
  ]}
 RigVMModel Construction Event Graph {nodes [
   n1 = &RigUnit_PrepareForExecution()
   &RigUnit_HierarchyImportFromSkeleton(NameSpace=None, bIncludeCurves=False, bIncludeMeshSockets=False, bIncludeVirtualBones=True, ExecutePin=@n1.ExecutePin)
  ]}
]
